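# Masked, skeletal-mesh-compatible material for the FenoWireframe vehicle wheel, adding glowing scanlines to rim and spokes.
# It takes attributes from MF_WheelParent_Simple and uses the rim mask's green and blue channels to blend an effect color into Base Color and dimmed Emissive.
# The scanlines come from a gradient texture panned in tiled screen space, with tunable color and intensity.
# The result passes through the shared vehicle damage and de-rez transition functions.
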Material /VehicleCosmetics/Wheels/FenoWireframe/Materials/M_Wheel_FenoWireframe
graph {
 properties {
  bAutomaticallySetUsageInEditor=false
  bUseMaterialAttributes=true
  bUsedWithMeshDeformer=true
  bUsedWithSkeletalMesh=true
  blendMode=BLEND_Masked
  pixelDepthOffsetMode=PDOM_Legacy
  usedShadingModels=Default Lit
 }
 nodes [
  n1 = MaterialFunctionCall() {materialFunction=&MF_WheelParent_Simple}
  n2 = GetMaterialAttributes(None=@n1.MaterialAttributes) {attributeGetTypes [#1, #2]}
  n3 = Add(A=ComponentMask(None=@n1.RimRGBAMask) {b=true}, B=ComponentMask(None=@n1.RimRGBAMask) {g=true})
  n4 = NamedRerouteUsage()
  n5 = ScalarParameter() {defaultValue=1, group=5_Special Effects, parameterName=Spokes Glow Intensity}
  n6 = ComponentMask(None=@n1.RimRGBAMask) {b=true}
  n7 = LinearInterpolate(A=NamedRerouteUsage(), B=NamedRerouteUsage(), Alpha=Add(A=Multiply(A=@n4, B=@n5), B=Add(B=@n6) {constA=0.05}))
  n8 = LinearInterpolate(A=@n2.Base Color, B=@n7, Alpha=@n3)
  n9 = Multiply(A=LinearInterpolate(A=@n2.Emissive Color, B=@n7, Alpha=@n3)) {constB=0.25}
  MaterialFunctionCall(In=MaterialFunctionCall(Material=SetMaterialAttributes!(MaterialAttributes=@n2.MaterialAttributes, Base Color=@n8, Emissive Color=@n9) {attributeSetTypes [#1, #2]}) {materialFunction=&MF_BaseVehicle_Damage}.Result) {materialFunction=&MF_VehicleDeRez_Transition}
  n10 = ScreenPosition()
  n11 = Panner(Coordinate=Multiply(A=ComponentMask(None=@n10.ViewportUV) {g=true, r=true}, B=ScalarParameter() {defaultValue=350, parameterName=Scanline Tiling})) {speedY=5}
  NamedRerouteDeclaration(None=Clamp(None=TextureSample(UVs=@n11) {texture=/VehicleCosmetics/Wheels/_SharedTextures/FX/LinearGradient_Balanced}.R)) {name=Scanlines_output, nodeColor=<rgba 0 0.0859 1 1>}
  NamedRerouteDeclaration(None=VectorParameter() {group=5_Special Effects, parameterName=EmissiveColor, defaultValue=<rgba 1 0 0 1>}.RGB) {name=EmissiveColor_output, nodeColor=<rgba 0.4062 0 1 1>}
  NamedRerouteDeclaration(None=Constant3Vector()) {name=NegativeColor_output, nodeColor=<rgba 0 1 0.0859 1>}
 ]
}
