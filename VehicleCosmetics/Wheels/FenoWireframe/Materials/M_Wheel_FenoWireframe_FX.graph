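# Translucent FX material for the FenoWireframe vehicle wheel, rendering rim and tire as an animated glow tinted by CustomColor.
# Vertex color red blends a rim branch, with a scrolling vertical gradient, and a tire branch, with a sideways-scrolling gradient shaped by packed mask channels.
# Screen-space scrolling scanlines form a named reroute for the spoke and tread glow terms.
# The blended intensity drives emissive color and, scaled by an Opacity parameter, transparency.
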Material /VehicleCosmetics/Wheels/FenoWireframe/Materials/M_Wheel_FenoWireframe_FX
graph {
 properties {
  bAutomaticallySetUsageInEditor=false
  bUseMaterialAttributes=true
  bUsedWithMeshDeformer=true
  bUsedWithSkeletalMesh=true
  blendMode=BLEND_Translucent
  pixelDepthOffsetMode=PDOM_Legacy
  usedShadingModels=Default Lit
 }
 nodes [
  n1 = VectorParameter() {group=5_Special Effects, parameterName=CustomColor, defaultValue=<rgba 1 0 0 1>}
  n2 = Panner(
   Coordinate=TextureCoordinate()
   Speed=AppendVector(A=Constant(), B=ScalarParameter() {defaultValue=0.75, group=5_Special Effects, parameterName=Speed})
  ) {speedY=-7.5}
  n3 = ScalarParameter() {defaultValue=2, group=5_Special Effects, parameterName=Detail Glow Intensity}
  n4 = NamedRerouteUsage()
  n5 = ScalarParameter() {defaultValue=1, group=5_Special Effects, parameterName=Spokes Glow Intensity}
  n6 = TextureSample() {samplerType=SAMPLERTYPE_Masks, texture=/VehicleCosmetics/Wheels/FenoWireframe/Textures/T_Rim_FenoWireframe_Mask}
  n7 = Add(
   A=Multiply(A=TextureSample(UVs=@n2) {texture=/VehicleCosmetics/Wheels/_SharedTextures/FX/Vert_Linear01_D}.RGB, B=@n3)
   B=Add(A=Multiply(A=@n4, B=@n5), B=Add(B=@n6.R) {constA=0.05})
  )
  n8 = TextureSample(UVs=Panner(Coordinate=TextureCoordinate()) {speedX=-1}) {texture=/VehicleCosmetics/Wheels/_SharedTextures/FX/Gradient_D}
  n9 = TextureSample() {samplerType=SAMPLERTYPE_LinearColor, texture=/VehicleCosmetics/Wheels/FenoWireframe/Textures/T_Tire_FenoWireframe_RGB}
  n10 = Multiply(A=@n3, B=Multiply(A=@n8.R, B=@n9.R))
  n11 = Multiply(A=@n9.R) {constB=0.25}
  n12 = ScalarParameter() {defaultValue=1, group=5_Special Effects, parameterName=Tread Glow Intensity}
  n13 = NamedRerouteUsage()
  n14 = Add(A=@n9.G, B=@n9.A)
  n15 = Add(
   A=LinearInterpolate(A=Add(A=@n10, B=@n11), B=Multiply(A=@n8.R) {constB=0.25}, Alpha=@n9.B)
   B=Add(A=Multiply(A=@n12, B=@n13), B=Add(B=@n14) {constA=0.05})
  )
  n16 = VertexColor()
  MakeMaterialAttributes!(
   BaseColor
   Metallic
   Specular
   Roughness
   Anisotropy
   EmissiveColor=LinearInterpolate(A=Multiply(A=@n1.RGB, B=@n7), B=Multiply(A=@n1.RGB, B=@n15), Alpha=@n16.R) {constA=1}
   Opacity=Multiply(A=LinearInterpolate(A=@n7, B=@n15, Alpha=@n16.R), B=ScalarParameter() {defaultValue=0.5, group=5_Special Effects, parameterName=Opacity})
   OpacityMask
   Normal
   Tangent
   WorldPositionOffset
   SubsurfaceColor
   ClearCoat
   ClearCoatRoughness
   AmbientOcclusion
   Refraction
   CustomizedUVs_0
   CustomizedUVs_1
   CustomizedUVs_2
   CustomizedUVs_3
   CustomizedUVs_4
   CustomizedUVs_5
   CustomizedUVs_6
   CustomizedUVs_7
   PixelDepthOffset
   ShadingModel
   Displacement
  )
  n17 = ScreenPosition()
  n18 = Panner(Coordinate=Multiply(A=ComponentMask(None=@n17.ViewportUV) {g=true, r=true}, B=ScalarParameter() {defaultValue=300, parameterName=Scanline Tiling})) {speedY=5}
  NamedRerouteDeclaration(None=Clamp(None=TextureSample(UVs=@n18) {texture=/VehicleCosmetics/Wheels/_SharedTextures/FX/LinearGradient_Balanced}.R)) {name=Scanlines_output, nodeColor=<rgba 0 0.0859 1 1>}
 ]
}
